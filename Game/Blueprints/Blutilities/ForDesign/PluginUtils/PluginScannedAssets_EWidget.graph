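# Editor utility widget listing Game Feature plugins and the primary assets each would scan.
# Populate runs Python to fetch all GameFeatureData assets and creates list items, whose assets come from SimulateScanningForPrimaryAssets.
# With bAutoScan on, a looping timer scans one item per tick, showing progress on the Populate button to spread the work out.
# Plugin and asset text filters hide non-matching items, and a checkbox tracks and toggles expand/collapse all.
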EditorUtilityWidgetBlueprint /Game/Blueprints/Blutilities/ForDesign/PluginUtils/PluginScannedAssets_EWidget
graphs [
 EventGraph {nodes [
   "On Clicked (btnPopulate)"!() -> (OutputDelegate, then, widget:User Widget Object Reference, event:Pointer Event Structure) {is=title K2Node_ComponentBoundEvent}
   > PopulateList()
   n1 = "On Text Committed (filterEntry_Asset)"!() -> (OutputDelegate, then, text:Text, commitMethod:ETextCommit Enum=Default) {is=title K2Node_ComponentBoundEvent}
   RunFilter(execute=[@n1.then, @n2.then])
   n2 = "On Text Committed (filterEntry_Plugin)"!() -> (OutputDelegate, then, text:Text, commitMethod:ETextCommit Enum=Default) {is=title K2Node_ComponentBoundEvent}
   "On Clicked (cbGrowShrink)"!() -> (OutputDelegate, then, widget:User Widget Object Reference, event:Pointer Event Structure) {is=title K2Node_ComponentBoundEvent}
   > handleGrowShrinkClick()
   Event Destruct()
   > Clear and Invalidate Timer by Handle!(execute, WorldContextObject:Object Reference, Handle=Get UpdateTimerHandle() -> (UpdateTimerHandle:Timer Handle Structure)) -> (then)
   Event Construct()
   > validateGrowShrinkButton()
  ]}
 PopulateList {nodes [
   PopulateList() {is=title K2Node_FunctionEntry}
   > Clear and Invalidate Timer by Handle(Handle=Get UpdateTimerHandle())
   > cleanupList()
   > n1 = Execute Python Script!(
    execute
    PythonScript="import TechArt.Tools.TechArtPluginUtils as plugin_utils\r\nimport importlib\r\nimportlib.reload(plugin_utils)\r\n\r\nall_gfd_datas = plugin_utils.getAllGameFeatureDatas()"
   ) -> (then, ReturnValue, all_gfd_datas:Array of Asset Data Structures) {outputs [all_gfd_datas]}
   > n2 = Set l_allGFDs(l_allGFDs:Array of Asset Data Structures=@n1.all_gfd_datas) -> (Output_Get:Array of Asset Data Structures)
   > n3 = For Each Loop(Array:Array of Asset Data Structures=@n2) -> (Array Element:Asset Data Structure)
   setSelected!(
    execute=@n3.Completed
    self:TA Base Editor Button Widget Object Reference=Get cbGrowShrink() -> (cbGrowShrink:TA Generic 01 Checkbox Object Reference)
    selected:bool=false
    throwEvent:bool=false
   ) -> (then)
   > validateGrowShrinkButton()
   > Branch(Condition=Get bAutoScan() -> (bAutoScan:bool))
   > n4 = Set Timer by Function Name!(
    execute
    Object:Object Reference
    FunctionName:String=ScanNextEntry
    Time:float=0.1
    bLooping:bool=true
    bMaxOncePerFrame:bool=false
    InitialStartDelay:float=0
    InitialStartDelayVariance:float=0
   ) -> (then, ReturnValue:Timer Handle Structure) {is=title K2Node_CallFunction}
   > Set UpdateTimerHandle(UpdateTimerHandle:Timer Handle Structure=@n4) -> (Output_Get:Timer Handle Structure)
   n5 = Cast To GameFeatureData!(execute=@n3.LoopBody, Object=GetAsset(InAssetData=@n3)) -> (then, CastFailed, AsGame Feature Data:Game Feature Data Object Reference) {is=title K2Node_DynamicCast}
   > n6 = Construct Object from Class(Class=/Game/Blueprints/Blutilities/ForTechArt/Data/TA_GenericListItemData.TA_GenericListItemData_C) -> (ReturnValue:TA Generic List Item Data Object Reference)
   > Set objects!(
    execute
    objects:Array of Object References=Make Array("[0]":Object Reference=@n5) -> (Array:Array of Object References)
    self:TA Generic List Item Data Object Reference=@n6
   ) -> (then, Output_Get:Array of Object References)
   > n7 = GetPluginName!(execute, GFD:Game Feature Data Object Reference=@n5) -> (then, PluginName:String)
   > Set displayString!(execute, displayString:String=@n7, self:TA Generic List Item Data Object Reference=@n6) -> (then, Output_Get:String)
   > n8 = addItem!(
    execute
    self:TA Generic 01 List Object Reference=Get listPlugins() -> (listPlugins:TA Generic 01 List Object Reference)
    dataObject:Object Reference=@n6
   ) -> (then, newItem:User Widget Object Reference, newData:Object Reference)
   > n9 = Cast To PluginScannedAssets_ListItem!(execute, Object=@n8) -> (then, CastFailed, AsPlugin Scanned Assets List Item:Plugin Scanned Assets List Item Object Reference) {is=title K2Node_DynamicCast}
   > Bind Event to On Expansion Changed(
    self:Expandable Area Object Reference=Get ExpandableArea_268!(self:Plugin Scanned Assets List Item Object Reference=@n9) -> (ExpandableArea_268:Expandable Area Object Reference)
    Delegate=Create Event()
   ) {is=title K2Node_AddDelegate}
   > Bind Event to Scan for Primary Assets(self:Plugin Scanned Assets List Item Object Reference=@n9, Delegate=Create Event()) {is=title K2Node_AddDelegate}
   GetObjectName(Object=GetOuterObject!(Object:Object Reference=@n5) -> (ReturnValue:Object Reference))
  ]}
 "PopulateScannedAssets " {nodes [
   n1 = "PopulateScannedAssets "!() -> (then, ListItem:Plugin Scanned Assets List Item Object Reference) {is=title K2Node_FunctionEntry}
   > n2 = SimulateScanningForPrimaryAssets!(
    execute
    GFD:Game Feature Data Object Reference=Get gameFeatureData!(self:Plugin Scanned Assets List Item Object Reference=@n1) -> (gameFeatureData:Game Feature Data Object Reference)
   ) -> (then, ReturnValue:Array of Soft Object Path Structures)
   > Set currentAssets!(execute, currentAssets:Array of Soft Object Path Structures=@n2, self:Plugin Scanned Assets List Item Object Reference=@n1) -> (then, Output_Get:Array of Soft Object Path Structures)
   > Return Node()
  ]}
 RunFilter {nodes [
   RunFilter() {is=title K2Node_FunctionEntry}
   > n1 = getListItems!(execute, self:TA Generic 01 List Object Reference=Get listPlugins()) -> (then, items:Array of User Widget Object References)
   > n2 = For Each Loop(Array:Array of User Widget Object References=@n1) -> (Array Element:User Widget Object Reference)
   > n3 = Cast To PluginScannedAssets_ListItem(Object=@n2)
   n4 = Get filterEntry_Asset() -> (filterEntry_Asset:Labeled Entry Widget Object Reference)
   n5 = Get filterEntry_Plugin() -> (filterEntry_Plugin:Labeled Entry Widget Object Reference)
   > n6 = Run Filter!(
    execute
    self:Plugin Scanned Assets List Item Object Reference=@n3
    plugin:String="To String (Text)"(InText=getText!(self:Labeled Entry Widget Object Reference=@n5) -> (text:Text))
    asset:String="To String (Text)"(InText=getText(self=@n4))
   ) -> (then, passed:bool)
   > SetVisibility(self=@n3, InVisibility=Select(Option 0=Collapsed, Option 1=Visible, Index=@n6))
  ]}
 ScanNextEntry {nodes [
   ScanNextEntry() {is=title K2Node_FunctionEntry}
   > n1 = getListItems(self=Get listPlugins())
   > Set Items(Items:Array of User Widget Object References=@n1) -> (Output_Get:Array of User Widget Object References)
   > n2 = Branch(Condition=Is Valid Index!(
     TargetArray:Array of User Widget Object References=Get Items() -> (Items:Array of User Widget Object References)
     IndexToTest:int=Get ScanIndex() -> (ScanIndex:int=0)
    ) -> (ReturnValue:bool))
   n3 = Get ScanIndex() -> (ScanIndex=0)
   n4 = Get Items()
   n5 = Append!(
    A="Populate List of Plugins (Scanning "
    B="To String (Integer)"(InInt=int + int(A=@n3))
    C:String=" of "
    D:String="To String (Integer)"(InInt=Length(TargetArray:Array of User Widget Object References=@n4))
    E:String=")"
   ) -> (ReturnValue)
   n6 = Get btnPopulate() -> (btnPopulate:TA Generic 01 Button Object Reference)
   > n7 = setLabel!(execute, self:TA Base Editor Button Widget Object Reference=@n6, label:String=@n5) -> (then)
   n8 = Cast To PluginScannedAssets_ListItem(
    execute=[@n7.then, @n11.then]
    Object="Get (a copy)"(Array:Array of User Widget Object References=Get Items(), Dimension 1=Get ScanIndex() -> (ScanIndex=0)) -> (Output:User Widget Object Reference)
   )
   > n9 = ScanForAssets!(execute, self:Plugin Scanned Assets List Item Object Reference=@n8, bExpand:bool=false) -> (then)
   Increment Int(" "=[@n9.then, @n8.CastFailed], Value=Get ScanIndex() -> (ScanIndex=0))
   n10 = setLabel(execute=@n2.else, self=@n6, label=Populate List of Plugins)
   Clear and Invalidate Timer by Handle(execute=[@n10.then, @n12.then], Handle=Get UpdateTimerHandle())
   n11 = Set label!(execute, label:String=@n5, self:TA Base Editor Button Widget Object Reference=@n6) -> (then, Output_Get:String)
   n12 = Set label(label=Populate List of Plugins, self=@n6)
  ]}
 cleanupList {nodes [
   cleanupList() {is=title K2Node_FunctionEntry}
   > n1 = Sequence()
   > n2 = getListItems(self=Get listPlugins())
   > n3 = For Each Loop with Break!(Exec:Exec, Array:Array of User Widget Object References=@n2, Break:Exec) -> (LoopBody:Exec, Array Element:User Widget Object Reference, Array Index:int, Completed:Exec)
   > n4 = Cast To PluginScannedAssets_ListItem(Object=@n3)
   > Unbind all Events from On Expansion Changed!(execute:Exec, self:Expandable Area Object Reference=Get ExpandableArea_268(self=@n4)) -> (then:Exec) {is=title K2Node_ClearDelegate}
   clearItems(execute=@n1.then_1, self:TA Generic 01 List Object Reference=Get listPlugins())
  ]}
 handleGrowShrinkClick {nodes [
   handleGrowShrinkClick() {is=title K2Node_FunctionEntry}
   n1 = Get cbGrowShrink()
   > Set l_newSelect(l_newSelect:bool=NOT Boolean(A=getSelected!(self:TA Base Editor Button Widget Object Reference=@n1) -> (isSelected:bool))) -> (Output_Get:bool)
   > n2 = getListItems(self=Get listPlugins())
   > n3 = For Each Loop with Break(Array=@n2)
   setSelected(execute=@n3.Completed, self=Get cbGrowShrink(), selected=Get l_newSelect() -> (l_newSelect:bool), throwEvent=false)
   n4 = Cast To PluginScannedAssets_ListItem(execute=@n3.LoopBody, Object=@n3)
   > SetIsExpanded!(execute, self:Expandable Area Object Reference=Get ExpandableArea_268(self=@n4), IsExpanded:bool=Get l_newSelect()) -> (then)
  ]}
 handleItemExpansionChanged {nodes [
   handleItemExpansionChanged!() -> (then, Area:Expandable Area Object Reference, bIsExpanded:bool=~) {is=title K2Node_FunctionEntry}
   > validateGrowShrinkButton()
  ]}
 validateGrowShrinkButton {nodes [
   validateGrowShrinkButton() {is=title K2Node_FunctionEntry}
   > n1 = Sequence!(execute) -> (then_0, then_1, then_2:Exec)
   > n2 = getListData!(execute, self:TA Generic 01 List Object Reference=Get listPlugins()) -> (then, data:Array of Object References)
   > SetIsEnabled(self=Get cbGrowShrink(), bInIsEnabled="integer > integer"(A=Length(TargetArray:Array of Object References=@n2)))
   n3 = getListItems(execute=@n1.then_1, self=Get listPlugins())
   > n4 = For Each Loop with Break(Array=@n3, Break=@n6.then)
   setSelected(execute=@n4.Completed, self=Get cbGrowShrink(), selected=Get l_areAnyExpanded() -> (l_areAnyExpanded:bool), throwEvent=false)
   n5 = Cast To PluginScannedAssets_ListItem(execute=@n4.LoopBody, Object=@n4)
   > Branch(Condition=GetIsExpanded!(self:Expandable Area Object Reference=Get ExpandableArea_268(self=@n5)) -> (ReturnValue:bool))
   > n6 = Set l_areAnyExpanded(l_areAnyExpanded:bool=true) -> (Output_Get:bool)
   n7 = getSelected(self=Get cbGrowShrink())
   SetToolTipText!(
    execute=@n1.then_2
    self:Widget Object Reference=Get cbGrowShrink()
    InToolTipText:"Text (by ref)"="To Text (String)"(InString=SelectString(A=Collapse All, B=Expand All, bPickA=@n7))
   ) -> (then)
  ]}
]
variables [
 bAutoScan {type=boolean}
 ScanIndex {type=integer}
 UpdateTimerHandle {type=TimerHandle}
]
widgets {
 info {rootWidgetClass=&VerticalBox, widgetCount=8}
 widgets [
  VerticalBox_0 {widgetClassPath=&VerticalBox, widget=VerticalBox_0}
  btnPopulate {
   widgetClassPath=&TA_Generic01_Button
   parent=VerticalBox_0
   widget=btnPopulate
   properties {label=Populate List of Plugins}
   slot {class=VerticalBoxSlot}
  }
  HorizontalBox_0 {
   widgetClassPath=&HorizontalBox
   parent=VerticalBox_0
   widget=HorizontalBox_0
   slot {class=VerticalBoxSlot, properties {padding=<ltrb 0 5 0 0>}}
  }
  cbGrowShrink {
   widgetClassPath=/Game/Blueprints/Blutilities/ForTechArt/Widgets/Buttons/TA_Generic01_Checkbox
   bIsVariable=true
   parent=HorizontalBox_0
   widget=cbGrowShrink
   properties {
    autoToggleSelection=false
    visibility=Visible
    buttonStyle_Default {
     clickedSlateSound {resourceObject=null}
     disabled {
      bIsDynamicallyLoaded=false
      drawAs=Image
      imageType=NoImage
      mirroring=NoMirror
      resourceName=None
      resourceObject=/Game/Blueprints/Blutilities/ForTechArt/ArtAssets/Textures/Checkbox/TA_HierarchyShowHide_Unchecked_Up
      tiling=NoTile
      imageSize=<16 16>
      margin=<ltrb 0.25 0.25 0.25 0.25>
      outlineSettings {
       bUseBrushTransparency=false
       roundingType=HalfHeightRadius
       width=0
       color {colorUseRule=UseColor_Specified, specifiedColor=<rgba 0 0 0 0>}
       cornerRadii=<0 0 0 1>
      }
      tintColor {colorUseRule=UseColor_Specified, specifiedColor=<rgba 0.2135 0.2135 0.2135 1>}
      uVRegion {bIsValid=false, max=<0 0>, min=<0 0>}
     }
     disabledForeground {colorUseRule=UseColor_Foreground, specifiedColor=<rgba 1 0 1 1>}
     hovered {
      bIsDynamicallyLoaded=false
      drawAs=Image
      imageType=NoImage
      mirroring=NoMirror
      resourceName=None
      resourceObject=/Game/Blueprints/Blutilities/ForTechArt/ArtAssets/Textures/Checkbox/TA_HierarchyShowHide_Unchecked_Over
      tiling=NoTile
      imageSize=<16 16>
      margin=<ltrb 0.25 0.25 0.25 0.25>
      outlineSettings {
       bUseBrushTransparency=false
       roundingType=HalfHeightRadius
       width=0
       color {colorUseRule=UseColor_Specified, specifiedColor=<rgba 0 0 0 0>}
       cornerRadii=<0 0 0 1>
      }
      tintColor {colorUseRule=UseColor_Specified, specifiedColor=<rgba 1 1 1 1>}
      uVRegion {bIsValid=false, max=<0 0>, min=<0 0>}
     }
     hoveredForeground {colorUseRule=UseColor_Foreground, specifiedColor=<rgba 1 0 1 1>}
     hoveredSlateSound {resourceObject=null}
     normal {
      bIsDynamicallyLoaded=false
      drawAs=Image
      imageType=NoImage
      mirroring=NoMirror
      resourceName=None
      resourceObject=&TA_HierarchyShowHide_Unchecked_Up
      tiling=NoTile
      imageSize=<16 16>
      margin=<ltrb 0.25 0.25 0.25 0.25>
      outlineSettings {
       bUseBrushTransparency=false
       roundingType=HalfHeightRadius
       width=0
       color {colorUseRule=UseColor_Specified, specifiedColor=<rgba 0 0 0 0>}
       cornerRadii=<0 0 0 1>
      }
      tintColor {colorUseRule=UseColor_Specified, specifiedColor=<rgba 1 1 1 1>}
      uVRegion {bIsValid=false, max=<0 0>, min=<0 0>}
     }
     normalForeground {colorUseRule=UseColor_Foreground, specifiedColor=<rgba 1 0 1 1>}
     normalPadding=<ltrb 0 0 0 0>
     pressed {
      bIsDynamicallyLoaded=false
      drawAs=Image
      imageType=NoImage
      mirroring=NoMirror
      resourceName=None
      resourceObject=&TA_HierarchyShowHide_Unchecked_Up
      tiling=NoTile
      imageSize=<16 16>
      margin=<ltrb 0.25 0.25 0.25 0.25>
      outlineSettings {
       bUseBrushTransparency=false
       roundingType=HalfHeightRadius
       width=0
       color {colorUseRule=UseColor_Specified, specifiedColor=<rgba 0 0 0 0>}
       cornerRadii=<0 0 0 1>
      }
      tintColor {colorUseRule=UseColor_Specified, specifiedColor=<rgba 0.4635 0.4635 0.4635 1>}
      uVRegion {bIsValid=false, max=<0 0>, min=<0 0>}
     }
     pressedForeground {colorUseRule=UseColor_Foreground, specifiedColor=<rgba 1 0 1 1>}
     pressedPadding=<ltrb 0 0 0 0>
     pressedSlateSound {resourceObject=null}
    }
    buttonStyle_Selected {
     clickedSlateSound {resourceObject=null}
     disabled {
      bIsDynamicallyLoaded=false
      drawAs=Image
      imageType=NoImage
      mirroring=NoMirror
      resourceName=None
      resourceObject=/Game/Blueprints/Blutilities/ForTechArt/ArtAssets/Textures/Checkbox/TA_HierarchyShowHide_Checked_Up
      tiling=NoTile
      imageSize=<16 16>
      margin=<ltrb 0.25 0.25 0.25 0.25>
      outlineSettings {
       bUseBrushTransparency=false
       roundingType=HalfHeightRadius
       width=0
       color {colorUseRule=UseColor_Specified, specifiedColor=<rgba 0 0 0 0>}
       cornerRadii=<0 0 0 1>
      }
      tintColor {colorUseRule=UseColor_Specified, specifiedColor=<rgba 0.2396 0.2396 0.2396 1>}
      uVRegion {bIsValid=false, max=<0 0>, min=<0 0>}
     }
     disabledForeground {colorUseRule=UseColor_Foreground, specifiedColor=<rgba 1 0 1 1>}
     hovered {
      bIsDynamicallyLoaded=false
      drawAs=Image
      imageType=NoImage
      mirroring=NoMirror
      resourceName=None
      resourceObject=/Game/Blueprints/Blutilities/ForTechArt/ArtAssets/Textures/Checkbox/TA_HierarchyShowHide_Checked_Over
      tiling=NoTile
      imageSize=<16 16>
      margin=<ltrb 0.25 0.25 0.25 0.25>
      outlineSettings {
       bUseBrushTransparency=false
       roundingType=HalfHeightRadius
       width=0
       color {colorUseRule=UseColor_Specified, specifiedColor=<rgba 0 0 0 0>}
       cornerRadii=<0 0 0 1>
      }
      tintColor {colorUseRule=UseColor_Specified, specifiedColor=<rgba 1 1 1 1>}
      uVRegion {bIsValid=false, max=<0 0>, min=<0 0>}
     }
     hoveredForeground {colorUseRule=UseColor_Foreground, specifiedColor=<rgba 1 0 1 1>}
     hoveredSlateSound {resourceObject=null}
     normal {
      bIsDynamicallyLoaded=false
      drawAs=Image
      imageType=NoImage
      mirroring=NoMirror
      resourceName=None
      resourceObject=&TA_HierarchyShowHide_Checked_Up
      tiling=NoTile
      imageSize=<16 16>
      margin=<ltrb 0.25 0.25 0.25 0.25>
      outlineSettings {
       bUseBrushTransparency=false
       roundingType=HalfHeightRadius
       width=0
       color {colorUseRule=UseColor_Specified, specifiedColor=<rgba 0 0 0 0>}
       cornerRadii=<0 0 0 1>
      }
      tintColor {colorUseRule=UseColor_Specified, specifiedColor=<rgba 1 1 1 1>}
      uVRegion {bIsValid=false, max=<0 0>, min=<0 0>}
     }
     normalForeground {colorUseRule=UseColor_Foreground, specifiedColor=<rgba 1 0 1 1>}
     normalPadding=<ltrb 0 0 0 0>
     pressed {
      bIsDynamicallyLoaded=false
      drawAs=Image
      imageType=NoImage
      mirroring=NoMirror
      resourceName=None
      resourceObject=&TA_HierarchyShowHide_Checked_Up
      tiling=NoTile
      imageSize=<16 16>
      margin=<ltrb 0.25 0.25 0.25 0.25>
      outlineSettings {
       bUseBrushTransparency=false
       roundingType=HalfHeightRadius
       width=0
       color {colorUseRule=UseColor_Specified, specifiedColor=<rgba 0 0 0 0>}
       cornerRadii=<0 0 0 1>
      }
      tintColor {colorUseRule=UseColor_Specified, specifiedColor=<rgba 0.4948 0.4948 0.4948 1>}
      uVRegion {bIsValid=false, max=<0 0>, min=<0 0>}
     }
     pressedForeground {colorUseRule=UseColor_Foreground, specifiedColor=<rgba 1 0 1 1>}
     pressedPadding=<ltrb 0 0 0 0>
     pressedSlateSound {resourceObject=null}
    }
   }
   slot {class=HorizontalBoxSlot, properties {verticalAlignment=VAlign_Top, padding=<ltrb 3 0 3 0>}}
  }
  filterEntry_Plugin {
   widgetClassPath=/Game/Blueprints/Blutilities/ForTechArt/Widgets/Misc/labeledEntry_Widget
   bIsVariable=true
   parent=HorizontalBox_0
   widget=filterEntry_Plugin
   properties {currentType=text, label=Plugin Filter}
   slot {class=HorizontalBoxSlot, properties {padding=<ltrb 10 0 0 0>, size {sizeRule=Fill, value=1}}}
  }
  filterEntry_Asset {
   widgetClassPath=&labeledEntry_Widget
   bIsVariable=true
   parent=HorizontalBox_0
   widget=filterEntry_Asset
   properties {currentType=text, label=Asset Filter}
   slot {class=HorizontalBoxSlot, properties {padding=<ltrb 10 0 0 0>, size {sizeRule=Fill, value=1}}}
  }
  Border_0 {
   widgetClassPath=&Border
   parent=VerticalBox_0
   widget=Border_0
   properties {brushColor=<rgba 1 1 1 0.15>, padding=<ltrb 4 4 4 4>}
   slot {class=VerticalBoxSlot, properties {padding=<ltrb 0 5 0 0>, size {sizeRule=Fill, value=1}}}
  }
  listPlugins {
   widgetClassPath=/Game/Blueprints/Blutilities/ForTechArt/Widgets/List/TA_Generic01_List
   bIsVariable=true
   parent=Border_0
   widget=listPlugins
   properties {
    allowMultiSelect=false
    listItemClass=&/PluginScannedAssets_ListItem.PluginScannedAssets_ListItem_C
   }
   slot {class=BorderSlot, properties {padding=<ltrb 4 4 4 4>}}
  }
 ]
}
